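# Level Sequence for the tac9200 sequence in PistolCanary that animates the tentacle boss.
# It possesses one BGA_Tentacle_C actor, bound as TentacleBoss, through a single skeletal animation track.
# The track holds two equal-length sections, one starting at frame zero and the other far back in negative frames.
# No master tracks or other bindings are present, and the display rate is 30 fps.
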LevelSequence /PistolCanary/Sequence/tac/tac9200/PistolCanary_tac9200_anim
bindings [TentacleBoss {
  id=#1
  possessed_class=BGA_Tentacle_C
  tracks [{
    type=MovieSceneSkeletalAnimationTrack
    sections [
     {type=MovieSceneSkeletalAnimationSection, end_frame=-7382, start_frame=-7502}
     {type=MovieSceneSkeletalAnimationSection, end_frame=120, start_frame=0}
    ]
   }]
 }]
display_rate {denominator=1, numerator=30}
tick_resolution {denominator=1, numerator=24000}
tracks []
