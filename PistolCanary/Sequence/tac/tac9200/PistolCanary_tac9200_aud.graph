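# Level Sequence that serves as the audio layer of the tac9200 sequence in PistolCanary, which involves the tentacle boss.
# A single root-level audio track holds one short section that starts just before frame zero, with volume and pitch left at their defaults.
# The TentacleBoss possessable is bound but has no tracks, and the section shows neither a sound asset nor an attach actor.
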LevelSequence /PistolCanary/Sequence/tac/tac9200/PistolCanary_tac9200_aud
bindings [TentacleBoss {id=#1, possessed_class=BGA_Tentacle_C, tracks []}]
display_rate {denominator=1, numerator=30}
tick_resolution {denominator=1, numerator=24000}
tracks [{
  type=MovieSceneAudioTrack
  display_name=Audio
  sections [{
    type=MovieSceneAudioSection
    end_frame=89
    start_frame=-2
    channels [
     Volume {type=MovieSceneScriptingFloatChannel, default=1}
     Pitch {type=MovieSceneScriptingFloatChannel, default=1}
     AttachActor {type=MovieSceneScriptingActorReferenceChannel}
    ]
   }]
 }]
